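# Niagara cosmetic effect for the Elastic female head that adds colored glows at both eyes and an eye trail.
# EyeGlow bursts one sprite per eye using spawn groups, and EyeTrail bursts one mesh, all at user-set offsets and gated by IsFXEnabled.
# A local color-hack module applies EyeColor with user multipliers and front-end versus gameplay brightness.
# Both emitters use camera offsets and set VisibilityTag when the head is hidden or ADS is active.
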NiagaraSystem /BRCosmetics/Characters/Player/Female/Medium/Heads/F_MED_Elastic_Head_01/FX/NS_Elastic_Eyetrail
dependencies {usedDataInterfaces [], usedDynamicInputs [], usedModules [&SystemState]}
emitters {
 EyeGlow {
  emitterUpdateScript [
   &EmitterState(Life Cycle Mode=::NewEnumerator1, Loop Behavior=::NewEnumerator1, Loop Duration Mode=::NewEnumerator1, UseLoopDelay=false)
   &SpawnBurst_Instantaneous(
    Spawn Count=&MakeCustomIntFromBool(Bool=HlslExpression() {hlslExpression="User.IsFXEnabled == 1"}, True INT=1, False INT=0)
    Use Spawn Probability=false
   )
   &SpawnBurst_Instantaneous(
    Spawn Count=&MakeCustomIntFromBool(Bool=HlslExpression() {hlslExpression="User.IsFXEnabled == 1"}, True INT=1, False INT=0)
    Spawn Group=1
    Use Spawn Probability=false
   ) {module=SpawnBurst_Instantaneous001}
  ]
  particleSpawnScript [
   /Niagara/Modules/Spawn/Initialization/InitializeParticle(
    Color=<rgba 0.6 0.005 0 0.4>
    Mesh Scale=<1.3 1.3 1.3>
    Module.Write Color=true
    Module.Write Lifetime=true
    Module.Write Mass=true
    Module.Write Material Random=true
    Module.Write Position=false
    Module.Write Scale=false
    Module.Write SpriteRotation=false
    Module.Write SpriteSize=true
    Position=&SimulationPosition()
    Sprite Size=&Vector2DFromFloat(Value=14)
   )
   &SystemLocation(MaskSpawnGroup=false, Offset=$User.RightEyeGlowLocation)
   &SystemLocation(MaskSpawnGroup=true, Offset=$User.LeftEyeGlowLocation, Spawn Group Mask=1) {module=SystemLocation001}
  ]
  particleUpdateScript [
   &ParticleState(Kill Particles When Lifetime Has Elapsed=false, Let Infinitely Lived Particles Die When Emitter Deactivates=true)
   &self:Elastic_Color_Hack(
    Brightest Value Color Multiplier:Vector3f=&VectorFromFloat(Value=&Multiply_Float(A=&FN_FrontEndFloatModifier(Gameplay=1, Front End=5), B=$User.EyeGlowMultiplier))
    Color:LinearColor=$User.EyeColor
    Darkest Value Multiplier=40
    Mid Range Color Multiplier:Vector3f=&VectorFromFloat(Value=&Multiply_Float(A=$User.EyeGlowMultiplier, B=&FN_FrontEndFloatModifier(Gameplay=1, Front End=5)))
    Scale Alpha:NiagaraFloat=$User.IsFXEnabled
   ) {module=Elastic_Color_Hack}
   &DynamicMaterialParameters(
    Index 1 Param 1=4
    Module.Param0WriteEnabled=true
    Module.Param0WriteEnabled1=true
    Module.Param1WriteEnabled=true
    Module.Param1WriteEnabled1=true
    Module.Param2WriteEnabled=true
    Module.Param2WriteEnabled1=true
    Module.Param3WriteEnabled=true
    Module.Param3WriteEnabled1=true
    Write Parameter Index 1=true
   )
   &CameraOffset(Camera Offset Amount=&FN_FrontEndFloatModifier(Gameplay=2.5, Front End=1.5))
   /SetVariables(Particles.VisibilityTag=&MakeCustomIntFromBool(Bool=&Boolean_OrOperation(Bool1=$User.bisHeadHidden, Bool2=$User.isADS), True INT=1, False INT=0)) {isSetParameters=true, module=SetVariables_#1}
  ]
  rendererClasses [&NiagaraSpriteRendererProperties]
  renderers [{rendererClass=&NiagaraSpriteRendererProperties}]
 }
 EyeTrail {
  emitterUpdateScript [
   &EmitterState(Life Cycle Mode=::NewEnumerator1, Loop Behavior=::NewEnumerator1, Loop Duration Mode=::NewEnumerator1, UseLoopDelay=false)
   &SpawnBurst_Instantaneous(
    Spawn Count=&MakeCustomIntFromBool(Bool=HlslExpression() {hlslExpression="User.IsFXEnabled == 1"}, True INT=1, False INT=0)
    Use Spawn Probability=false
   )
  ]
  particleSpawnScript [
   &InitializeParticle(
    Color=<rgba 0.6 0.005 0 0.4>
    Lifetime=0
    Mesh Scale=<1.3 1.3 1.3>
    Module.Write Color=true
    Module.Write Lifetime=true
    Module.Write Mass=true
    Module.Write Material Random=true
    Module.Write Position=false
    Module.Write Scale=false
    Module.Write SpriteRotation=false
    Module.Write SpriteSize=true
    Position=&SimulationPosition()
    Sprite Size=&Vector2DFromFloat(Value=16.5)
   )
   &SystemLocation(MaskSpawnGroup=false, Offset=$User.EyetrailLocation)
   &InitialMeshOrientation(Mesh Orientation Mode=::NewEnumerator2, Orientation Vector=<0 0 1>, Use Rotation Vector=false)
  ]
  particleUpdateScript [
   &ParticleState(Kill Particles When Lifetime Has Elapsed=false)
   &self:Elastic_Color_Hack(
    Brightest Value Color Multiplier:Vector3f=&VectorFromFloat(Value=&Multiply_Float(A=&FN_FrontEndFloatModifier(Gameplay=5, Front End=5), B=$User.EyeTrailMultiplier))
    Color:LinearColor=$User.EyeColor
    Darkest Value Multiplier=40
    Mid Range Color Multiplier:Vector3f=&VectorFromFloat(Value=&Multiply_Float(A=$User.EyeTrailMultiplier, B=&FN_FrontEndFloatModifier(Gameplay=5, Front End=15)))
    Scale Alpha:NiagaraFloat=$User.IsFXEnabled
   ) {module=Elastic_Color_Hack}
   &CameraOffset(Camera Offset Amount=1)
   /SetVariables(Particles.VisibilityTag=&MakeCustomIntFromBool(Bool=&Boolean_OrOperation(Bool1=$User.bisHeadHidden, Bool2=$User.isADS), True INT=1, False INT=0)) {isSetParameters=true, module=SetVariables_#1}
  ]
  rendererClasses [&NiagaraMeshRendererProperties]
  renderers [{rendererClass=&NiagaraMeshRendererProperties, rendererIndex=0}]
 }
}
summary {
 systemName=NS_Elastic_Eyetrail
 emitters [
  {emitterName=EyeGlow, rendererClasses [&NiagaraSpriteRendererProperties]}
  {emitterName=EyeTrail, rendererClasses [&NiagaraMeshRendererProperties]}
 ]
 userVariables [
  User.bIsFrontEnd {type=NiagaraBool}
  User.bIsFrontEndPreview {type=NiagaraBool}
  User.bisHeadHidden {type=NiagaraBool}
  User.EyeColor {type=LinearColor, description="", defaultValue {struct=LinearColor, value=<rgba 1 1 1 1>}}
  User.EyeGlowMultiplier {type=NiagaraFloat, defaultValue=0.8, description=""}
  User.EyetrailLocation {type=Vector3f, description="", defaultValue {struct=Vector3f, value=<3.0037 7.2775 0>}}
  User.EyeTrailMultiplier {type=NiagaraFloat, defaultValue=1, description=""}
  User.isADS {type=NiagaraBool}
  User.IsFXEnabled {type=NiagaraFloat, defaultValue=1, description=""}
  User.LeftEyeGlowLocation {type=Vector3f, description="", defaultValue {struct=Vector3f, value=<3.1368 7.6309 -3.3638>}}
  User.RightEyeGlowLocation {type=Vector3f, description="", defaultValue {struct=Vector3f, value=<3.1368 7.6309 3.3638>}}
 ]
}
